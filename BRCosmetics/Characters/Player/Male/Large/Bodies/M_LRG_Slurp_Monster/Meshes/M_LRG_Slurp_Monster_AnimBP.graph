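# Animation Blueprint for the large male Slurp Monster body cosmetic, adjusting hip pouch and grenade bones as the thighs bend.
# The event graph maps each thigh's yaw delta from reference pose to a clamped crouch alpha.
# The AnimGraph copies the attached parent's pose, applies additive bone offsets weighted by those alphas, then layers half-weight Anim Dynamics on the left grenade and both upper pouch bones for secondary motion.
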AnimBlueprint /BRCosmetics/Characters/Player/Male/Large/Bodies/M_LRG_Slurp_Monster/Meshes/M_LRG_Slurp_Monster_AnimBP
target_skeleton=/Game/Characters/Player/Male/Large/Bodies/M_LRG_Slurp_Monster/Meshes/M_LRG_Slurp_Monster_Skeleton
graphs [
 AnimGraph {nodes [
   n1 = Get thigh_l_crouch() -> (thigh_l_crouch:double)
   n2 = "Transform (Modify) Bone - Bone: dyn_grenade_02_l"(
    ComponentPose=Local To Component(LocalPose=Copy Pose From Mesh() {bUseAttachedParent=true})
    Alpha=@n1
    Translation="0.000,-3.808,3"
    Rotation="0,0,0"
    Scale="1,1,1"
   ) {
    rotationMode=BMM_Additive
    rotationSpace=BCS_BoneSpace
    translationMode=BMM_Additive
    translationSpace=BCS_BoneSpace
    alphaBoolBlend {blendOption=Linear}
    boneToModify {boneName=dyn_grenade_02_l}
    translation {y=-3.808, z=3}
   }
   n3 = "Transform (Modify) Bone - Bone: dyn_hippouch_01_l"(ComponentPose=@n2, Alpha=@n1, Translation="-0.144,-3.178,0.000", Rotation="0.000,-13.572,0.000", Scale="1,1,1") {
    rotationMode=BMM_Additive
    rotationSpace=BCS_BoneSpace
    translationMode=BMM_Additive
    translationSpace=BCS_BoneSpace
    alphaBoolBlend {blendOption=Linear}
    boneToModify {boneName=dyn_hippouch_01_l}
    rotation {yaw=-13.5722}
    translation=<-0.144 -3.178>
   }
   n4 = "Transform (Modify) Bone - Bone: dyn_hippouch_02_l"(ComponentPose=@n3, Alpha=@n1, Translation="3.456,-5.163,8.264", Rotation="87,0.000,-20", Scale="1,1,1") {
    rotationMode=BMM_Additive
    rotationSpace=BCS_BoneSpace
    translationMode=BMM_Additive
    translationSpace=BCS_BoneSpace
    alphaBoolBlend {blendOption=Linear}
    boneToModify {boneName=dyn_hippouch_02_l}
    rotation {pitch=87, roll=-20}
    translation=<3.456 -5.163 8.264>
   }
   n5 = Get thigh_r_crouch() -> (thigh_r_crouch:double)
   n6 = "Transform (Modify) Bone - Bone: dyn_hippouch_01_r"(ComponentPose=@n4, Alpha=@n5, Translation="0,0,0", Rotation="0,0,0", Scale="1,1,1") {
    rotationMode=BMM_Additive
    rotationSpace=BCS_BoneSpace
    translationMode=BMM_Additive
    translationSpace=BCS_BoneSpace
    alphaBoolBlend {blendOption=Linear}
    boneToModify {boneName=dyn_hippouch_01_r}
   }
   n7 = "Transform (Modify) Bone - Bone: dyn_hippouch_02_r"(ComponentPose=@n6, Alpha=@n5, Translation="3.456,-5.163,8.264", Rotation="90,0,0", Scale="1,1,1") {
    rotationMode=BMM_Additive
    rotationSpace=BCS_BoneSpace
    translationMode=BMM_Additive
    translationSpace=BCS_BoneSpace
    alphaBoolBlend {blendOption=Linear}
    boneToModify {boneName=dyn_hippouch_02_r}
    rotation {pitch=90}
    translation=<3.456 -5.163 8.264>
   }
   n8 = Anim Dynamics - Bone: dyn_grenade_02_l(ComponentPose=@n7, Alpha=0.5, ExternalForce="0,0,0") {
    angularDampingOverride=1
    bGravityOverrideInSimSpace=true
    bOverrideAngularDamping=true
    gravityScale=0.35
    simulationSpace=RootRelative
    alphaBoolBlend {blendOption=Linear}
    boundBone {boneName=dyn_grenade_02_l}
    componentLinearAccScale=<0.2 0.2 0.2>
    componentLinearVelScale=<0.2 0.2 0.2>
    physicsBodyDefinitions [{
      sphereCollisionRadius=0
      boundBone {boneName=dyn_grenade_02_l}
      boxExtents=<10 10 5>
      constraintSetup {angularTargetAxis=AxisZ, twistAxis=AxisZ, angularLimitsMax=<2 0 0>, angularLimitsMin=<-2 -45 0>, angularTarget=<0 0 0>}
      localJointOffset=<0 0 -5>
     }]
   }
   n9 = Anim Dynamics - Bone: dyn_hippouch_01_l(ComponentPose=@n8, Alpha=0.5, ExternalForce="0,0,0") {
    angularDampingOverride=1
    bGravityOverrideInSimSpace=true
    bOverrideAngularDamping=true
    gravityScale=0.35
    simulationSpace=RootRelative
    alphaBoolBlend {blendOption=Linear}
    boundBone {boneName=dyn_hippouch_01_l}
    componentLinearAccScale=<0.2 0.2 0.2>
    componentLinearVelScale=<0.2 0.2 0.2>
    physicsBodyDefinitions [{
      sphereCollisionRadius=0
      boundBone {boneName=dyn_hippouch_01_l}
      boxExtents=<10 10 5>
      constraintSetup {angularTargetAxis=AxisZ, twistAxis=AxisZ, angularLimitsMax=<2 0 0>, angularLimitsMin=<-2 -45 0>, angularTarget=<0 0 0>}
      localJointOffset=<0 0 -5>
     }]
   }
   n10 = Anim Dynamics - Bone: dyn_hippouch_01_r(ComponentPose=@n9, Alpha=0.5, ExternalForce="0,0,0") {
    angularDampingOverride=1
    bGravityOverrideInSimSpace=true
    bOverrideAngularDamping=true
    gravityScale=0.35
    simulationSpace=RootRelative
    alphaBoolBlend {blendOption=Linear}
    boundBone {boneName=dyn_hippouch_01_r}
    componentLinearAccScale=<0.2 0.2 0.2>
    componentLinearVelScale=<0.2 0.2 0.2>
    physicsBodyDefinitions [{
      sphereCollisionRadius=0
      boundBone {boneName=dyn_hippouch_01_r}
      boxExtents=<10 10 5>
      constraintSetup {angularTargetAxis=AxisZ, twistAxis=AxisZ, angularLimitsMax=<2 0 0>, angularLimitsMin=<-2 -45 0>, angularTarget=<0 0 0>}
      localJointOffset=<0 0 -5>
     }]
   }
   Output Pose(Result=Component To Local(ComponentPose=@n10)) {layerGroup=None}
  ]}
 EventGraph {nodes [
   Event BlueprintUpdateAnimation()
   n1 = GetPartSkeletalMeshComponent()
   > Set thigh_l_crouch(thigh_l_crouch:double=MapRangeClamped(
     Value=GetDeltaTransformFromRefPose!(self=@n1, BoneName=thigh_l, BaseName=None) -> (
      ReturnValue_Location:Vector
      ReturnValue_Rotation_Roll:float
      ReturnValue_Rotation_Pitch:float
      ReturnValue_Rotation_Yaw:float
      ReturnValue_Scale:Vector
     ).ReturnValue_Rotation_Yaw
     InRangeA=-10
     InRangeB=-90
    )) -> (Output_Get:double)
   > Set thigh_r_crouch(thigh_r_crouch:double=MapRangeClamped(Value=GetDeltaTransformFromRefPose(self=@n1, BoneName=thigh_r, BaseName=None).ReturnValue_Rotation_Yaw, InRangeA=-10, InRangeB=-90)) -> (Output_Get:double)
  ]}
]
variables [
 thigh_l_crouch {type=number}
 thigh_r_crouch {type=number}
]
