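# Physics asset for the face accessory part of the large male Slurp Monster character, giving its dangling "slurp" bones physical motion.
# Small capsule bodies sit on the head, a short slurp bone and a two-bone long slurp chain.
# Constraints tie the short bone and first long bone to the head, and the second long bone to the first, all with narrow swing and tighter twist limits.
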PhysicsAsset /BRCosmetics/Characters/Player/Male/Large/Bodies/M_LRG_Slurp_Monster/Meshes/Parts/M_LRG_Slurp_Monster_FaceAcc_Physics
bodies {
 dyn_longslurp_01 [{shapeType=Capsule, length=1, radius=1.5, center=<1 0 0>, rotation=<rot 90 0 0>}]
 dyn_longslurp_02 [{shapeType=Capsule, length=1, radius=1.5, center=<1 0 0>, rotation=<rot 90 0 0>}]
 dyn_shortslurp_01 [{shapeType=Capsule, length=1, radius=1.5, center=<1 0 0>, rotation=<rot 90 0 0>}]
 head [{shapeType=Capsule, length=2, radius=1, rotation=<rot 90 0 0>}]
}
constraints [
 {bone1Name=dyn_shortslurp_01, bone2Name=head, swing1LimitDegrees=20, swing2LimitDegrees=20, twistLimitDegrees=5}
 {bone1Name=dyn_longslurp_01, bone2Name=head, swing1LimitDegrees=20, swing2LimitDegrees=20, twistLimitDegrees=5}
 {bone1Name=dyn_longslurp_02, bone2Name=dyn_longslurp_01, swing1LimitDegrees=20, swing2LimitDegrees=20, twistLimitDegrees=5}
]
