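# Niagara cosmetic effect for dripping slime on the Slurp Monster player body, red variant.
# Drips_Start spawns drips probabilistically across seven skeletal mesh bones or sockets with per-group offsets and sizes, stretching them as they swell.
# It raises conditional death events, apparently feeding Drips_Drop, whose sprites fall with fading source movement.
# A per-group visibility tag drives rendering and event gating, switching one group when User.bisHandHidden is set.
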NiagaraSystem /BRCosmetics/Characters/Player/Male/Large/Bodies/M_LRG_Slurp_Monster/VFX/NS_SlurpMon_Drips_Red
dependencies {
 usedDataInterfaces [
  &NiagaraDataInterfaceArrayFloat
  &NiagaraDataInterfaceArrayInt32
  &NiagaraDataInterfaceArrayFloat3
  &NiagaraDataInterfaceVector2DCurve
  &NiagaraDataInterfaceVectorCurve
  &NiagaraDataInterfaceCurve
  &NiagaraDataInterfaceSkeletalMesh
 ]
 usedDynamicInputs []
 usedModules [
  &self:SystemSpawnScript.SystemScriptSource.SystemScriptGraph.NiagaraNodeAssignment_0.SetVariables_#1
  &SystemState
 ]
}
emitters {
 Drips_Drop {
  emitterUpdateScript [&EmitterState()]
  particleSpawnScript [
   /Niagara/Modules/Spawn/Initialization/InitializeParticle(
    Lifetime=&RandomRangeFloat(Minimum=0.6, Maximum=0.8, Evaluation Type=::NewEnumerator1)
    Module.Write Color=true
    Module.Write Lifetime=true
    Module.Write Mass=true
    Module.Write Material Random=true
    Module.Write Position=false
    Module.Write Scale=false
    Module.Write SpriteRotation=false
    Module.Write SpriteSize=true
    Position=<-10 0 0>
    Sprite Size=&Vector2DFromFloat(Value=&RandomRangeFloat(Minimum=11, Maximum=14, Evaluation Type=::NewEnumerator1))
   )
   &InheritVelocity(Module.Limit Speed=false, Velocity Limit=0, Velocity Scale=<0.3 0.3 0.3>)
  ]
  particleUpdateScript [
   &ParticleState()
   &ScaleSpriteSize(Scale Factor=&Vector2DFromFloat(Value=&FloatFromCurve(FloatCurve=DataInterface() {propertyValues {Curve {keys [
          {arriveTangent=5.4603, interpMode=RCIM_Cubic, time=0, value=0.1}
          {arriveTangent=-0.1, interpMode=RCIM_Cubic, leaveTangent=-0.1, time=0.1819, value=0.9885}
          <at 1 0 RCIM_Cubic>
         ]}}})))
   &InheritSourceMovement(
    Applied Force Scale=<0 0 1>
    Applied Position Scale=&VectorFromFloat(Value=&FloatFromCurve(FloatCurve=DataInterface() {propertyValues {Curve {keys [{arriveTangent=-3.2322, interpMode=RCIM_Cubic, leaveTangent=-3.2322, tangentMode=RCTM_User, time=0, value=1}, <at 1 0 RCIM_Cubic>]}}}))
    ApplyForce=false
    ApplyPosition=true
    ApplyPositionFalloff=false
   )
   &AccelerationForce(Acceleration=&VectorFromCurve(VectorCurve=DataInterface() {propertyValues {
       XCurve {keys [<at 0 0 RCIM_Cubic>,<at 0.3 0 RCIM_Cubic>]}
       YCurve {keys [<at 0 0 RCIM_Cubic>,<at 0.3 0 RCIM_Cubic>]}
       ZCurve {keys [<at 0 0 RCIM_Cubic>, {interpMode=RCIM_Cubic, leaveTangent=-1141.1395, time=0.3, value=-600}]}
      }}))
   &ScaleColor(
    Scale Alpha=&FloatFromCurve(FloatCurve=DataInterface() {propertyValues {Curve {keys [<at 0 0>,<at 0.2 1>]}}})
    Scale RGB=&VectorFromCurve(VectorCurve=DataInterface() {propertyValues {XCurve {keys [<at 0 1>,<at 1 1>]}, YCurve {keys [<at 0 1>,<at 1 1>]}, ZCurve {keys [<at 0 1>,<at 1 1>]}}})
    ScaleA=true
    ScaleRGB=true
   )
   &CameraOffset(Camera Offset Amount=6)
   &SolveForcesAndVelocity(Clamp Velocity=false, Limit Acceleration=false)
  ]
  rendererClasses [&NiagaraSpriteRendererProperties]
  renderers [{rendererClass=&NiagaraSpriteRendererProperties}]
 }
 Drips_Start {
  emitterUpdateScript [
   &EmitterState()
   &SpawnBurst_Instantaneous(Spawn Count=&RandomRangeInt(Minimum=0, Maximum=1), Use Spawn Probability=false) {enabled=false}
   &SpawnRate(Spawn Probability=&RandomRangeFloat(Minimum=0, Maximum=1), SpawnRate=3.3, Use Spawn Probability=true)
  ]
  particleSpawnScript [
   &FN_SetSpawnGroup(Number of Spawn Groups=7)
   &InitializeParticle(
    Lifetime=&RandomRangeFloat(Minimum=1.8, Maximum=2.1, Evaluation Type=::NewEnumerator1)
    Module.Write Color=true
    Module.Write Lifetime=true
    Module.Write Mass=true
    Module.Write Material Random=true
    Module.Write Position=false
    Module.Write Scale=false
    Module.Write SpriteRotation=false
    Module.Write SpriteSize=true
    Sprite Size=&Vector2DFromFloat(Value=&Multiply_Float(
      A=&RandomRangeFloat(Minimum=7, Maximum=11)
      B=&SelectFloatFromArray(Direct Array Index=$Particles.SpawnGroup, Float Selection Array=DataInterface() {propertyValues {FloatData [0.8,0.8,0.7,0.7,0.7,0.7,0.65]}})
     ))
   )
   /SetVariables(Particles.VisibilityTag=&SelectIntFromArray(Direct Array Index=$Particles.SpawnGroup, Int Selection Array=DataInterface() {propertyValues {IntData [0,0,0,0,1,0,0]}})) {isSetParameters=true}
  ]
  particleUpdateScript [
   &ParticleState()
   &SkeletalMeshLocation(
    Bone / Socket Index=$Particles.SpawnGroup
    Bone Sampling Mode=::NewEnumerator4
    Inherit Velocity=false
    MaskSpawnGroup=false
    Sampled Position Offset=&SelectVectorFromArray(
     Direct Array Index=$Particles.SpawnGroup
     Vector Selection Array=DataInterface() {propertyValues {FloatData [<25 -2 0>, <10 -2 0>, <-20 2.5 0>, <-5 4 0>, <-7 10 10>, <0 0 0>, <-10 3 0>]}}
    )
    Sampled Position Offset Coordinate Space:ENiagaraMeshTransforms=::NewEnumerator3
    Skeletal Mesh=$System.Skeletal Mesh
   )
   &ScaleSpriteSize(Scale Factor=&Vector2DFromCurve(
     Vector2Curve=DataInterface() {propertyValues {
       XCurve {keys [
         <at 0 1.15 RCIM_Cubic>
         {arriveTangent=-0.8581, interpMode=RCIM_Cubic, leaveTangent=-0.8581, time=0.6037, value=1.14}
         {arriveTangent=-0.3533, interpMode=RCIM_Cubic, leaveTangent=-0.3533, time=0.8741, value=0.4}
         <at 1 1>
        ]}
       YCurve {keys [
         <at 0 0 RCIM_Cubic>
         {arriveTangent=2.8602, interpMode=RCIM_Cubic, leaveTangent=2.8602, time=0.6037, value=0.9885}
         {arriveTangent=2.0477, interpMode=RCIM_Cubic, leaveTangent=2.0477, time=0.8741, value=2.5}
         <at 1 1.8>
        ]}
      }}
     ScaleCurve=<1 1>
    ))
   &AccelerationForce(Acceleration=&VectorFromCurve(
     VectorCurve=DataInterface() {propertyValues {
       XCurve {keys [<at 0 0 RCIM_Cubic>,<at 0.3 0 RCIM_Cubic>]}
       YCurve {keys [<at 0 0 RCIM_Cubic>,<at 0.3 0 RCIM_Cubic>]}
       ZCurve {keys [<at 0 -1 RCIM_Cubic>, {interpMode=RCIM_Cubic, leaveTangent=-205.935, time=0.3, value=-20}]}
      }}
     Scale Curve=&VectorFromFloat(Value=1)
    ))
   &SolveForcesAndVelocity(Clamp Velocity=false, Limit Acceleration=false)
   &ScaleColor(
    Scale Alpha=&FloatFromCurve(FloatCurve=DataInterface() {propertyValues {Curve {keys [<at 0 0>,<at 0.1 1>,<at 0.9 1>,<at 1 0>]}}})
    Scale RGB=&VectorFromCurve(VectorCurve=DataInterface() {propertyValues {XCurve {keys [<at 0 1>,<at 1 1>]}, YCurve {keys [<at 0 1>,<at 1 1>]}, ZCurve {keys [<at 0 1>,<at 1 1>]}}})
    ScaleA=true
    ScaleRGB=true
   )
   &CameraOffset(Camera Offset Amount=10)
   /SetVariables(Particles.VisibilityTag=&MakeCustomIntFromBool(
     Bool=&SetBoolByIntComparison(A=$Particles.SpawnGroup, B=5, Comparison Type=::NewEnumerator2)
     True INT=&MakeCustomIntFromBool(Bool=$User.bisHandHidden, True INT=2, False INT=$Particles.Initial.VisibilityTag)
     False INT=$Particles.Initial.VisibilityTag
    )) {isSetParameters=true}
   &GenerateDeathEvent(Event Condition:NiagaraBool=&SetBoolByIntComparison(A=$Particles.VisibilityTag, B=2, Comparison Type=::NewEnumerator5))
  ]
  rendererClasses [&NiagaraSpriteRendererProperties]
  renderers [
   {rendererClass=&NiagaraSpriteRendererProperties}
   {rendererClass=&NiagaraSpriteRendererProperties, rendererIndex=1}
  ]
 }
}
summary {
 systemName=NS_SlurpMon_Drips_Red
 emitters [
  {emitterName=Drips_Start, rendererClasses [&NiagaraSpriteRendererProperties]}
  {emitterName=Drips_Drop, rendererClasses [&NiagaraSpriteRendererProperties]}
 ]
 userVariables [User.bisHandHidden {type=NiagaraBool}]
}
